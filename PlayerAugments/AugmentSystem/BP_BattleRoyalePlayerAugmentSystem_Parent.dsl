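; Parent controller component for Battle Royale player augments, granting augment selections on timers during a match.
; After the playlist is ready, delayed initialization finds the controller's granting, roll, persistence and UI components, with retries.
; Game phase or aircraft events start the repeating timer; clients drive UI and sync the first grant to a replicated server timestamp.
; Choice requests combine any newly claimed unlock with rolled augments, waiting a tick to avoid duplicates.
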(event EventEndPlay
  (params (EndPlayReason :type EEndPlayReason%20Enum :default "Destroyed"))
  ;[0.3467 0.3095 1] End Play
  (|CallParentFunction "Destroyed")
  (CallFunction|ClearAircraftBinds))

(event EventBeginPlay
  ;[0.3467 0.3095 1] Begin Play
  (|CallParentFunction)
  (Class|FortAsyncActionCurrentPlaylistReady|PlaylistDataReadyAsync
    (:OnReady
      (if (CallFunction|IsBRAugmentSystemEnabled)
        (CallFunction|Initialize)))))

(event Custom|InitializeGrantingComponent
  ;[0.3467 0.3095 1] Begin Play
  ;;[1 1 1] Init Granting Component
  (bind _returnvalue (Class|FortControllerComponentAugmentGrantingSystem|GetAugmentGrantingComponentfromController (Variables|Default|GetController)))
  (bind _output_get (Variables|Default|SetGrantComp _returnvalue))
  (Utilities|IsValid _output_get
    (:Is%20Valid
      _)
    (:Is%20Not%20Valid
      (Utilities|FlowControl|Sequence
        (:then_0
          _)
        (:then_1
          (Utilities|FlowControl|DoOnce false
            (:Completed
              (Utilities|FlowControl|Delay 5.0))) ; only try again 1 time to avoid infinite loop
        ))
      (CallFunction|BlueprintAugmentLog "Battle Royale Augment System failed to initialize granting component")))
  (Utilities|FlowControl|Sequence
    (:then_0
      ;;;[1 1 1] Bind Events
      (EventDispatchers|AssignOnAuthorityRequestAugmentChoices (Variables|Default|GetGrantComp) (EventDispatchers|CreateEvent :!event Custom|OnAuthorityRequestAugmentChoices_Event))
      (EventDispatchers|BindEventtoOnSelectedAugmentforGrant (Variables|Default|GetGrantComp) (EventDispatchers|CreateEvent :!event Custom|OnPlayerChoseAugment)))
    (:then_1
      (Class|FortControllerComponentAugmentGrantingSystem|AuthorityGrantFreeReroll (Variables|Default|GetGrantComp)))))

(event Custom|Initialize%20Roll%20Component
  ;[0.3467 0.3095 1] Begin Play
  ;;[1 1 1] Init Roll Component
  (bind _output_get (Variables|Default|SetPlayerRollComp (Actor|GetComponentbyClass (Variables|Default|GetController) "/Script/PlayerAugmentsCodeRuntime.FortControllerComponent_AugmentDynamicRollPlayerComponent")))
  (Networking|SwitchHasAuthority
    (:Authority
      (Utilities|IsValid _output_get
        (:Is%20Not%20Valid
          (Utilities|FlowControl|Sequence
            (:then_0
              _)
            (:then_1
              (Utilities|FlowControl|DoOnce false
                (:Completed
                  (Utilities|FlowControl|Delay 5.0))) ; only try again 1 time to avoid infinite loop
            ))
          (CallFunction|BlueprintAugmentLog "Battle Royale Augment System failed to initialize Roll Component"))))) ; Roll component is only on the server ATM\nSo only retry if on server.
)

(event Custom|Initialize%20Persitsence%20Component
  ;[0.3467 0.3095 1] Begin Play
  ;;[1 1 1] Init Persistence Component
  (bind _output_get (Variables|Default|SetPersistenceComponent (Actor|GetComponentbyClass (Variables|Default|GetController) "/Script/PlayerAugmentsCodeRuntime.FortControllerComponent_PlayerAugmentPersistence")))
  (Utilities|IsValid _output_get
    (:Is%20Not%20Valid
      (Utilities|FlowControl|Sequence
        (:then_0
          _)
        (:then_1
          (Utilities|FlowControl|DoOnce false
            (:Completed
              (Utilities|FlowControl|Delay 5.0))) ; only try again 1 time to avoid infinite loop
        ))
      (Utilities|Casting|CastToPlayerController (Components|GetOwner)) ; Only log if this a player, \nBOTS do not have persistence!
      (CallFunction|BlueprintAugmentLog "Battle Royale Augment System failed to initialize Persistence Component"))))

(event Custom|Initialize
  ;[0.3467 0.3095 1] Begin Play
  ;;[1 1 1] Initialize
  (Utilities|FlowControl|Delay 1.0) ; this used to be a passive GA that spawned\nafter controller components. But this had\nto be refactored to a controller component.\nThere appears to be a potential race condition\nwith components being created. So, we initialize\nthis after some delay, instead of begin play. \nToe ensure all components have been added \nto the controller before we attempt to get them\nhere.
  (if (not (Variables|Default|GetInitialized))
    (bind _outcontroller (CallFunction|GetAugmentController)
      (:Valid
        (Variables|Default|SetController _outcontroller)
        (Utilities|FlowControl|Sequence
          (:then_0
            (CallFunction|InitializeGrantingComponent)
            (CallFunction|InitializeRollComponent)
            (CallFunction|InitializePersitsenceComponent)
            (CallFunction|InitializeUIComponent))
          (:then_1
            (Utilities|IsValid _outcontroller
              (:Is%20Valid
                (if (Editor|InEditororPIE)
                  (bind _returnvalue (Development|GetConsoleVariableBoolValue "PlayerAugments.EnablePieDebugging"))
                  (if _returnvalue
                    _)
                  (else
                    _))
                (CallFunction|SetupAircraftBinds)
                (CallFunction|InitializeComponents)
                (CallFunction|TryStartTimerfromActivate))))
          (:then_5
            (Variables|Default|SetInitialized true))))
      (:Invalid
        (CallFunction|BlueprintAugmentLog "FAILED TO GET CONTROLLER. This should not happen.
Is this no longer a controller component?")))))

(event Custom|Initialize%20UI%20Component
  ;[0.3467 0.3095 1] Begin Play
  ;;[1 1 1] Init UIEvents
  (Utilities|SwitchHasAuthority
    (:Remote
      (Variables|Default|SetUIEventComp (Actor|GetComponentbyClass (Variables|Default|GetController) "/Script/PlayerAugmentsCodeRuntime.FortControllerComponent_AugmentUIEvents"))
      (Utilities|IsValid (Variables|Default|GetUIEventComp)
        (:Is%20Valid
          _)
        (:Is%20Not%20Valid
          (Utilities|FlowControl|Sequence
            (:then_0
              (if (< (Variables|UI|GetNumInitUIAttempts) 5)
                (CallFunction|BlueprintAugmentLog "B R Augment System Failed to get UIEvents component" "Log" true)
                (Math|Integer|IncrementInt (Variables|UI|GetNumInitUIAttempts)
                  (:%20%20
                    (Utilities|FlowControl|Delay 5.0))))))))
      (CallFunction|BlueprintAugmentLog "B R Augment System Acquired UI Component" "Log" true))))

(event Custom|InitializeComponents)

(event Custom|StartSelectionGrantTimer
  (bind _currentaugmentindex (Variables|Default|GetCurrentAugmentIndex))
  (bind _map_augcounttotime (Variables|Default|GetMAP_AugCountToTime))
  (bind (_value _returnvalue) (Utilities|Map|Find _map_augcounttotime _currentaugmentindex))
  (bind _uieventcomp (Variables|Default|GetUIEventComp))
  ;[1 0.8801 0.3073] Start Timer - Find the time based on the Augment index count an start a timer.
  (if (Variables|ParentFeatures|GetEnableAugmentGrantTimer)
    (if (CallFunction|CanGrantMoreAugments)
      (Utilities|FlowControl|Sequence
        (:then_0
          (if _returnvalue
            (bind _overridetime (CallFunction|GetClientOverrideFirstGrantTime))
            (bind _returnvalue_2 (Utilities|Time|SetTimerbyEvent (EventDispatchers|CreateEvent :!event Custom|GrantAugmentSelectionToken) (Math|Float|SelectFloat _overridetime (ScalableFloat|GetValueatLevel _value) (> _overridetime 0))))
            (Variables|Default|SetTimer_AugmentGrant _returnvalue_2)
            (CallFunction|AuthoritySetReplicatedGrantStartTime)
            (Variables|Default|SetGrantingStarted true)))
        (:then_1
          ;;[1 1 1] Have the UI timer Update
          (Utilities|FlowControl|DoOnce false
            (:Completed
              (Networking|CanExecuteCosmeticEvents
                (:True
                  (bind _returnvalue_3 (Utilities|Time|SetTimerbyEvent (delegate Custom|UpdateTimer) (Variables|Default|GetTimerUpdateTime) true))
                  (Variables|Default|SetUITimerUpdateHandle _returnvalue_3)))))))
      (else
        ;;[1 1 1] No More Grants
        (Networking|CanExecuteCosmeticEvents
          (:True
            (Utilities|IsValid _uieventcomp
              (:Is%20Valid
                (Class|FortControllerComponentAugmentUIEvents|NotifyGrantingTimersFinished _uieventcomp)
                (Utilities|Time|ClearandInvalidateTimerbyHandle (Variables|Default|GetUITimerUpdateHandle))))))))))

(event Custom|TryStartTimerFromActivate
  ;[1 1 1] Try start timer from activate
  (if (not (Variables|Default|GetUseGamephaseToStartGrantTimer))
    ;;[1 1 1] old way
    (Default|IsinWarmUp
      (:false
        (CallFunction|StartSelectionGrantTimer)))))

(event Custom|UpdateTimer
  (bind _uieventcomp (Variables|Default|GetUIEventComp))
  ;[1 0.8801 0.3073] Start Timer - Find the time based on the Augment index count an start a timer.
  ;;[1 1 1] Have the UI timer Update
  (Utilities|IsValid _uieventcomp
    (:Is%20Valid
      (Class|FortControllerComponentAugmentUIEvents|UpdatedTimerProgress _uieventcomp (- (ScalableFloat|GetValueatLevel (Utilities|Map|Find (Variables|Default|GetMAP_AugCountToTime) (Variables|Default|GetCurrentAugmentIndex))) (Utilities|Time|GetTimerElapsedTimebyHandle (Variables|Default|GetTimer_AugmentGrant))) (/ (Utilities|Time|GetTimerElapsedTimebyHandle (Variables|Default|GetTimer_AugmentGrant)) (ScalableFloat|GetValueatLevel (Utilities|Map|Find (Variables|Default|GetMAP_AugCountToTime) (Variables|Default|GetCurrentAugmentIndex))))))))

(event Custom|OnPlayerChoseAugment
  (params (SelectedAugment :type Fort%20Player%20Augment%20Item%20Definition%20Object%20Reference))
  ;[1 1 1] On Player Chose Choice
  (CallFunction|TryIncrementUnlockRollCount))

(event EventOnServerAugmentGrantTimestampReplicated
  (params (ServerGrantStartTimestamp :type Float%20%28single-precision%29))
  ;[0 1 0.7254] Client Timer Synchonization
  (Networking|SwitchHasAuthority
    (:Remote
      (if (== (Variables|Default|GetFirstAugmentGrantIndex) (Variables|Default|GetCurrentAugmentIndex))
        (if (Utilities|Time|IsValidTimerHandle (Variables|Default|GetTimer_AugmentGrant))
          (bind _overridetime (CallFunction|GetClientOverrideFirstGrantTime))
          (if (> _overridetime 0)
            (CallFunction|BlueprintAugmentLog "Augment Passive is synchronizing client timer to server timer")
            (Utilities|Time|ClearandInvalidateTimerbyHandle (Variables|Default|GetTimer_AugmentGrant))
            (CallFunction|StartSelectionGrantTimer)))))))

(event Custom|Authority_SetReplicatedGrantStartTime
  ;[0 1 0.7254] Client Timer Synchonization
  (Networking|SwitchHasAuthority
    (:Authority
      (if (== (Variables|Default|GetFirstAugmentGrantIndex) (Variables|Default|GetCurrentAugmentIndex))
        (CallFunction|AuthoritySetServerGrantingStartedTimestamp (GameState|GetServerWorldTimeSeconds (Game|GetGameState)))))))

(event Custom|GrantAugmentSelectionToken
  ;[0.8818 0 1] Grant Augment Granter - Increment the Augment count and call give item on the server before restarting the timer.
  (Math|Integer|IncrementInt (Variables|Default|GetCurrentAugmentIndex)
    (:%20%20
      (Utilities|FlowControl|Sequence
        (:then_0
          (Utilities|SwitchHasAuthority
            (:Authority
              (bind _returnvalue (Class|FortControllerComponentAugmentGrantingSystem|GetAugmentGrantingComponentfromController (Utilities|Casting|CastToController (Components|GetOwner))))
              (Class|FortControllerComponentAugmentGrantingSystem|AuthorityGrantNewAugment _returnvalue))))
        (:then_1
          ;;[1 1 1] start timer both on server and client, so client can update UI
          (Utilities|FlowControl|Delay 0.2)
          (CallFunction|StartSelectionGrantTimer))))))

(event EventOnPossedPawnChanged
  (params (OldPawn :type Pawn%20Object%20Reference) (NewPawn :type Pawn%20Object%20Reference)))

(event Custom|OnAuthorityRequestAugmentChoices_Event
  (params (IgnoreAugments :type Array%20of%20Fort%20Player%20Augment%20Item%20Definition%20Object%20References))
  ;[0.3931 0.2083 1] Provide choices when player needs augment choices
  (Utilities|IsValid (Variables|Default|GetPlayerRollComp)
    (:Is%20Valid
      (Utilities|FlowControl|DelayUntilNextTick) ; delay so we get an inventory update\nif player is stacking augment grant tokens. otherwise\nthe roll system will not yet know they already\nhave the augment.Letting them roll a duplicate.
      (Utilities|Array|Clear (Variables|Default|GetRollIgnoreItems))
      (Utilities|Array|Clear (Variables|Default|GetConvertedChoices)) ; clear this before we unock anything
      (Variables|Default|SetUnlocked%20Augment 0)
      (Utilities|FlowControl|Sequence
        (:then_0
          (bind _returnvalue (CallFunction|TryClaimUnlock))
          (Utilities|IsValid (Utilities|Struct|BreakFortAugmentPassiveUnlockResult _returnvalue)
            (:Is%20Valid
              (if (not (.bAlreadyInInventory (Utilities|Struct|BreakFortAugmentPassiveUnlockResult _returnvalue)))
                (bind _output_get (Variables|Default|SetUnlocked%20Augment (Utilities|Struct|BreakFortAugmentPassiveUnlockResult _returnvalue)))
                (Utilities|Array|Add (Variables|Default|GetRollIgnoreItems) _output_get)
                (Utilities|Array|Add (Variables|Default|GetConvertedChoices) (Utilities|Struct|MakePlayerAugmentChoiceDatum (Utilities|Struct|BreakFortAugmentPassiveUnlockResult _returnvalue) "(GameplayTags=((TagName=\"PlayerAugments.Choice.NewUnlock\")))"))))))
        (:then_1
          (for _array_element IgnoreAugments
            (Utilities|Array|Add (Variables|Default|GetRollIgnoreItems) _array_element))
          (bind _returnvalue_2 (Class|FortControllerComponentDynamicRollPlayerComponent|AuthorityRollChoices (Variables|Default|GetPlayerRollComp) (Math|Integer|SelectInt 1 2 (Utilities|IsValid (Variables|Default|GetUnlocked%20Augment) :!class /Script/Engine.KismetSystemLibrary)) (Variables|Default|GetRollIgnoreItems)))
          (for _array_element_2 _returnvalue_2
            (Utilities|Array|Add (Variables|Default|GetConvertedChoices) (Utilities|Struct|MakePlayerAugmentChoiceDatum (Utilities|Casting|CastToFortPlayerAugmentItemDefinition (Utilities|Struct|BreakFortDynamicRollResult _array_element_2)))))
          (Class|FortControllerComponentAugmentGrantingSystem|AuthoritySetAvailableAugmentChoices (Variables|Default|GetGrantComp) (Variables|Default|GetConvertedChoices))
          (Utilities|Array|Clear (Variables|Default|GetConvertedChoices)))))
    (:Is%20Not%20Valid
      (Development|PrintString "FAILED TO FIND PLAYER ROLL COMP"))))

(event Custom|SetupAircraftBinds
  ;[0 1 0.02065] Handle getting onto the plane after warmup
  (Utilities|FlowControl|Sequence
    (:then_0
      (bind (_aircraftcomponent _targetpawnorcontrollerisvalid) (Default|GetAircraftComponent))
      (if _targetpawnorcontrollerisvalid
        (Utilities|IsValid _aircraftcomponent
          (:Is%20Valid
            (EventDispatchers|BindEventtoOnAircraftStateChange _aircraftcomponent (EventDispatchers|CreateEvent :!event Custom|HandleAircraftStateChange))))))
    (:then_1
      (Utilities|IsValid (Fortnite|Athena|GameState|GetGameStateAthena)
        (:Is%20Valid
          _)
        (:Is%20Not%20Valid
          (Utilities|FlowControl|Delay 1.0)))
      (bind _returnvalue (GameplayEventRouter|GetEventRouter "/Script/FortniteGame.FortGameState" (Components|GetOwner)))
      (bind (_asynctaskproxy _payload) (GameplayEventRouter|StartListeningToStatefulEvent _returnvalue "/Script/FortniteGame.FortGamePhaseUpdatedEvent")
        (:then
          (Variables|Default|SetGamePhaseUpdatedAsyncTask _asynctaskproxy))
        (:OnEventReceived
          _)
        (:OnSavedEventStateExists
          _))
      (CallFunction|HandleGamePhaseChanged (Utilities|Struct|BreakFortGamePhaseUpdatedEvent _payload)))))

(event Custom|ClearAircraftBinds
  ;[0 1 0.02065] Handle getting onto the plane after warmup
  (Utilities|FlowControl|Sequence
    (:then_0
      (bind (_aircraftcomponent _targetpawnorcontrollerisvalid) (Default|GetAircraftComponent))
      (if _targetpawnorcontrollerisvalid
        (Utilities|IsValid _aircraftcomponent
          (:Is%20Valid
            (EventDispatchers|UnbindEventfromOnAircraftStateChange _aircraftcomponent (EventDispatchers|CreateEvent :!event Custom|HandleAircraftStateChange))))))
    (:then_1
      (bind _gamephaseupdatedasynctask (Variables|Default|GetGamePhaseUpdatedAsyncTask)
        (:then
          (GameplayEventRouter|StopListeningtoStatefulEvent _gamephaseupdatedasynctask))
        (:else)))))

(event Custom|HandleAircraftStateChange
  ;[0 1 0.02065] Handle getting onto the plane after warmup
  (Utilities|FlowControl|Delay 0.1)
  (Default|IsinAircraft
    (:true
      (if (not (Variables|Default|GetUseGamephaseToStartGrantTimer))
        (CallFunction|StartSelectionGrantTimer)))))

(event Custom|HandleGamePhaseChanged
  (params (GamePhase :type EAthenaGamePhase%20Enum))
  ;[0 1 0.02065] Handle getting onto the plane after warmup
  (if (Variables|Default|GetUseGamephaseToStartGrantTimer)
    (switch Utilities|FlowControl|Switch|SwitchonEAthenaGamePhase GamePhase
      (:None
        (if (not (Editor|InEditororPIE))
          (Debugging|LogString "NewEnumerator0" "NewEnumerator0" "Augment Passive Encountered Game Phase NONE in non-editor build." true)))
      (:Aircraft
        _)
      (:SafeZones
        _))
    (if (not (Variables|Default|GetGrantingStarted))
      (CallFunction|StartSelectionGrantTimer))))

(fn CanGrantMoreAugments
  (return (<= (Variables|Default|GetCurrentAugmentIndex) (ScalableFloat|GetValueatLevel (Variables|Default|GetTotalTimeEarnedAugmentCount)))))

(fn GetClientOverrideFirstGrantTime
  (bind _returnvalue (CallFunction|GetServerGrantingStartedTimestamp))
  (bind _returnvalue_2 (Fortnite|Athena|GameState|GetGameStateAthena))
  (bind _returnvalue_3 (GameState|GetServerWorldTimeSeconds _returnvalue_2))
  (bind _currentaugmentindex (Variables|Default|GetCurrentAugmentIndex))
  (bind _map_augcounttotime (Variables|Default|GetMAP_AugCountToTime))
  (bind (_value _returnvalue_4) (Utilities|Map|Find _map_augcounttotime _currentaugmentindex))
  (bind _returnvalue_5 (ScalableFloat|GetValueatLevel _value))
  (bind _returnvalue_6 (+ _returnvalue _returnvalue_5))
  (bind _returnvalue_7 (- _returnvalue_6 _returnvalue_3))
  (Networking|SwitchHasAuthority
    (:Authority
      _)
    (:Remote
      (if (== (Variables|Default|GetFirstAugmentGrantIndex) (Variables|Default|GetCurrentAugmentIndex))
        (if (> (CallFunction|GetServerGrantingStartedTimestamp) 0)
          (if _returnvalue_4
            (if (> _returnvalue_7 0)
              (return (Math|Float|Clamp(Float) _returnvalue_7 0.0 _returnvalue_5))
              (else
                _))
            (else
              _))
          (return -1.0)
          (else
            _))
        (else
          _)))))

(fn GetAircraftComponent
  ;[1 1 1] Tries to get the aircraft component off of player controller
  (bind _outcontroller (CallFunction|GetAugmentController)
    (:Valid
      (Utilities|IsValid (Actor|GetComponentbyClass _outcontroller "/Script/FortniteGame.FortControllerComponent_Aircraft")
        (:Is%20Valid
          (return (Actor|GetComponentbyClass _outcontroller "/Script/FortniteGame.FortControllerComponent_Aircraft") true))
        (:Is%20Not%20Valid
          (return 0 true))))
    (:Invalid
      (return 0 false))))
