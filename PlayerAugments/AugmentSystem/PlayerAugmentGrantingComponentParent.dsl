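; Parent Blueprint component for granting player augments, handling local feedback when augment selections are pressed.
; The local equip-selection event plays a 2D sound taken from GetEquipSound.
; The matching unequip-selection event is present but has no logic.
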(event EventOnLocalEquipAugmentSelectionPressed
  (Audio|SpawnSound2D (Variables|Default|GetEquipSound)))

(event EventOnLocalUnequipAugmentSelectionPressed)
